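; Thrown projectile for the HeavyLamp bladed weapon, which lets the player pawn switch to other weapons after the throw.
; On BeginPlay it binds to the pawn's weapon equip and unequip delegates, unbinding them on EndPlay.
; Unequipping holsters the weapon; equipping one without the Lobster weapon tag unholsters it, forcing the equip animation that holstering blocked.
; The construction script applies hue shift, red/blue switch and emissive intensity variables to the blade materials.
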(event EventBeginPlay
  ; BeginPlay
  ; Bind events to weapon equipped/unequipped to allow other weapons to be used after throwing the projectile
  (|CallParentFunction)
  (bind _playerpawn (Variables|BPSetVariables|GetPlayerPawn)
    (:then
      (FortPawn|BindEventtoOnWeaponEquippedDelegate _playerpawn (EventDispatchers|CreateEvent :!event Custom|AttemptUnHolsterWeapon))
      (FortPawn|BindEventtoOnWeaponUnEquippedDelegate _playerpawn (EventDispatchers|CreateEvent :!event Custom|HolsterOnUnequipped)))
    (:else)))

(event Custom|AttemptUnHolsterWeapon
  (params (NewWeapon :type Fort%20Weapon%20Object%20Reference) (PrevWeapon :type Fort%20Weapon%20Object%20Reference))
  ; AttemptUnHolsterWeapon
  ; Allow other weapons to be used after throwing projectile
  (if (not (Tags|WeaponDataHasMatchingGameplayTag NewWeapon (Variables|Default|GetLobsterWeaponTag)))
    (bind _playerpawn (Variables|BPSetVariables|GetPlayerPawn)
      (:then
        (Weapon|SetHolsterWeapon _playerpawn 0 false true "realm") ; force play the equip animation since we blocked \nit by holstering during the Unequip event
      )
      (:else))))

(event Custom|HolsterOnUnequipped
  ; HolsterOnUnEquipped
  (bind _playerpawn (Variables|BPSetVariables|GetPlayerPawn)
    (:then
      (Weapon|SetHolsterWeapon _playerpawn 0 true false "realm"))
    (:else)))

(event EventEndPlay
  (params (EndPlayReason :type EEndPlayReason%20Enum :default "Destroyed"))
  ; EndPlay
  ; Cleanup: unbind events
  (Utilities|FlowControl|Sequence
    (:then_0
      (bind _playerpawn (Variables|BPSetVariables|GetPlayerPawn)
        (:then
          (FortPawn|UnbindEventfromOnWeaponEquippedDelegate _playerpawn (EventDispatchers|CreateEvent :!event Custom|AttemptUnHolsterWeapon))
          (FortPawn|UnbindEventfromOnWeaponUnEquippedDelegate _playerpawn (EventDispatchers|CreateEvent :!event Custom|HolsterOnUnequipped)))
        (:else)))
    (:then_1
      (|CallParentFunction EndPlayReason))))

(fn ConstructionScript
  (bind _blade (Variables|Default|Getblade))
  (|CallParentFunction)
  (Rendering|Material|SetScalarParameterValueonMaterials _blade "HueColorShift" (Variables|Default|GetBladeHueColorShift))
  (Rendering|Material|SetScalarParameterValueonMaterials _blade "RedBlueSwitch" (Variables|Default|GetBladeRedBlueSwitch))
  (Rendering|Material|SetScalarParameterValueonMaterials _blade "EmissiveIntensity" (Variables|Default|GetBladeEmissiveIntensity)))
